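# Sound cue for a tree being destroyed, played as a demolition impact with close-range attenuation.
# The first two quality slots mix two randomized, pitch-modulated layers: tree-final impacts shaped by a fading volume envelope, and wood debris sounds.
# The other two slots play a single optimized tree-destroy wave with random pitch.
# An object-impact concurrency set is assigned to limit simultaneous instances.
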SoundCue /Figment_S10/Sounds/Fort_Impact_Sounds/Demolition/Wood/Demo_Tree_Destroyed_Cue
root=QualityLevel_0
nodes [
 {
  id=Enveloper_3
  children [Modulator_2]
  properties {
   pitchCurve {editorCurveData {keys [{time=0}]}}
   volumeCurve {editorCurveData {keys [<at 0 1>, {arriveTangent=-0.268, interpMode=RCIM_Cubic, leaveTangent=-0.268, time=1.358, value=0.9856}, <at 3.8687 -0.0368 RCIM_Cubic>]}}
  }
 }
 {id=Mixer_9, children [Enveloper_3,Modulator_11], properties {inputVolume [1,1]}}
 {id=Modulator_0, children [WavePlayer_0], properties {pitchMax=1.1, pitchMin=0.9}}
 {id=Modulator_11, children [Random_13], properties {pitchMax=1.1, pitchMin=0.9}}
 {id=Modulator_2, children [Random_7], properties {pitchMax=1.07, pitchMin=0.92}}
 {id=QualityLevel_0, children [Mixer_9,Mixer_9,Modulator_0,Modulator_0]}
 {id=Random_13, children [WavePlayer_37, WavePlayer_38], properties {weights [1,1]}}
 {id=Random_7, children [WavePlayer_42, WavePlayer_43], properties {weights [1,1]}}
 {id=WavePlayer_0, properties {soundWaveAssetPtr=&/Tree_Destroy_opt}}
 {id=WavePlayer_37, properties {soundWaveAssetPtr=/Game/Sounds/Fort_Impact_Sounds/Demolition/Wood/final_wood_new_03}}
 {id=WavePlayer_38, properties {soundWaveAssetPtr=/Game/Sounds/Fort_Impact_Sounds/Demolition/Wood/final_wood_new_02}}
 {id=WavePlayer_42, properties {soundWaveAssetPtr=/Figment_S10/Sounds/Fort_Impact_Sounds/Demolition/Fort_Impact_TreeFinal_02}}
 {id=WavePlayer_43, properties {soundWaveAssetPtr=/Figment_S10/Sounds/Fort_Impact_Sounds/Demolition/Fort_Impact_TreeFinal_01}}
]
properties {
 attenuationSettings=/Figment_S10/Sounds/Attenuation/Fort_Demolition_Close_Att
 maxDistance=2000
 soundClassObject=/Game/Sounds/FortSoundClassesAndModes/Fort_ImpactSounds_DuckDefault_PSM
 volumeMultiplier=0.4
 concurrencyOverrides {resolutionRule=PreventNew}
 concurrencySet [&CONC_Object_Impacts_FourMax]
}
