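# Sound cue for tree impacts in the demolition impact sound set, playing big tree hit samples.
# A quality level node selects one of four branches, stepping down from four equally weighted random variations to three, two, then a single wave.
# Each branch passes through a modulator for random pitch variation, slightly wider on the single-wave branch.
# The cue uses close demolition attenuation, an impact sound class and a shared object-impact concurrency asset.
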SoundCue /Figment_S10/Sounds/Fort_Impact_Sounds/Demolition/Wood/Demo_Tree_Impact_Cue
root=QualityLevel_0
nodes [
 {id=Modulator_0, children [Random_0], properties {pitchMax=1.07, pitchMin=0.92}}
 {id=Modulator_1, children [Random_1], properties {pitchMax=1.07, pitchMin=0.92}}
 {id=Modulator_2, children [Random_6], properties {pitchMax=1.07, pitchMin=0.92}}
 {id=Modulator_3, children [WavePlayer_5], properties {pitchMax=1.1, pitchMin=0.9}}
 {id=QualityLevel_0, children [Modulator_2,Modulator_0,Modulator_1,Modulator_3]}
 {id=Random_0, children [WavePlayer_0,WavePlayer_2,WavePlayer_1], properties {weights [1,1,1]}}
 {id=Random_1, children [WavePlayer_3,WavePlayer_4], properties {weights [1,1]}}
 {id=Random_6, children [WavePlayer_35, WavePlayer_38, WavePlayer_37, WavePlayer_39], properties {weights [1,1,1,1]}}
 {id=WavePlayer_0, properties {soundWaveAssetPtr=&/tree_big_hit_01}}
 {id=WavePlayer_1, properties {soundWaveAssetPtr=&/tree_big_hit_03}}
 {id=WavePlayer_2, properties {soundWaveAssetPtr=&/tree_big_hit_07}}
 {id=WavePlayer_3, properties {soundWaveAssetPtr=&tree_big_hit_01}}
 {id=WavePlayer_35, properties {soundWaveAssetPtr=&tree_big_hit_01}}
 {id=WavePlayer_37, properties {soundWaveAssetPtr=&tree_big_hit_03}}
 {id=WavePlayer_38, properties {soundWaveAssetPtr=&tree_big_hit_07}}
 {id=WavePlayer_39, properties {soundWaveAssetPtr=&/tree_big_hit_08}}
 {id=WavePlayer_4, properties {soundWaveAssetPtr=&tree_big_hit_07}}
 {id=WavePlayer_5, properties {soundWaveAssetPtr=&tree_big_hit_01}}
]
properties {
 attenuationSettings=/Figment_S10/Sounds/Attenuation/Fort_Demolition_Close_Att
 maxDistance=2000
 soundClassObject=/Game/Sounds/FortSoundClassesAndModes/Fort_ImpactSounds_DuckDefault_PSM
 volumeMultiplier=0.5
 concurrencyOverrides {resolutionRule=PreventNew}
 concurrencySet [&CONC_Object_Impacts_FourMax]
}
